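# One-shot stereo MetaSound source for the Battle Pass purchase button hover sound, in the Fort_UI sound class.
# On play, a Random Get picks from the Sounds array, which holds one shortened hover wave by default, and triggers the wave player via a zero-time delay.
# Both channels are scaled by MasterVolume, and playback end fires OnFinished; an Array Shuffle node sits unconnected.
# Overridden concurrency allows one instance, stopping the oldest.
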MetaSoundSource /SeasonPassHub/Audio/MSS_BattlePass_Purchase_Button_Hover
inputs [
 MasterVolume {type=Float}
 Sounds {type=WaveAsset:Array}
 UE.Source.OnPlay {type=Trigger}
]
interfaces [UE.OutputFormat.Stereo:1.0, UE.Source.OneShot:1.0, UE.Source:1.1]
outputs [
 "UE.OutputFormat.Stereo.Audio:0" {type=Audio}
 "UE.OutputFormat.Stereo.Audio:1" {type=Audio}
 UE.Source.OneShot.OnFinished {type=Trigger}
]
pages [Default {
  input_defaults {MasterVolume=0.736, UE.Source.OnPlay=null, Sounds [&/UI_BattlePass_Purchase_Hover_Lite_8_Shortened]}
  nodes [
   Array.Shuffle:WaveAsset:Array() {also {defaults {Enable Shared State="<MetaSoundBuilderResult.SUCCEEDED: 0>"}}}
   ArrayRandomGet_0 = Array.Random Get:WaveAsset:Array(In Array=^Sounds, Next=^UE.Source.OnPlay)
   TriggerDelay_0 = UE.Trigger Delay(In=@ArrayRandomGet_0.On Next) {also {defaults {Delay Time=0}}}
   WavePlayer_0 = UE.Wave Player:Stereo(Wave Asset=@ArrayRandomGet_0.Value, Play=@TriggerDelay_0.Out) {also {version [1,1]}}
   Multiply_0 = UE.Multiply:Audio by Float(AdditionalOperands=^MasterVolume, PrimaryOperand=@WavePlayer_0.Out Left)
   Multiply_1 = UE.Multiply:Audio by Float(AdditionalOperands=^MasterVolume, PrimaryOperand=@WavePlayer_0.Out Right)
  ]
  links [
   {dst="@UE.OutputFormat.Stereo.Audio:0", src=Multiply_0.Out}
   {dst="@UE.OutputFormat.Stereo.Audio:1", src=Multiply_1.Out}
   {dst="@UE.Source.OneShot.OnFinished", src=WavePlayer_0.On Finished}
  ]
 }]
properties {
 assetImportData=&self:AssetImportData
 bOverrideConcurrency=true
 numChannels=2
 outputFormat=Stereo
 sampleRate=48000
 soundClassObject=&Fort_UI
 concurrencyOverrides {
  bVolumeScaleCanRelease=true
  maxCount=1
  resolutionRule=StopOldest
  voiceStealReleaseTime=0.215
  volumeScale=0
  volumeScaleAttackTime=0.1
  volumeScaleReleaseTime=0.1
 }
 concurrencySet [/Game/Sounds/Concurrency_Rules/FrontEnd_UI_Menu_Max_01_Stop_Oldest_Hover_Tick]
}
